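# Translucent unlit Niagara mesh-particle material for the ball in the PigeonChart emote, with animated noise coloring and a dissolve.
# Opposite-panning noises, offset by a noise-based UV disturbance, combine with a rounded, sine-shaped V mask to blend two colors.
# That blend multiplies a Fresnel color gradient and particle color, then passes through emissive adjustments.
# A dynamic-parameter Dissolve drives HeightLerp over fast-panning noise, multiplied by particle alpha for opacity.
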Material /BRCosmetics/Effects/Fort_Effects/Emotes/PigeonChart/Material/M_Emote_PigeonChartBall
graph {
 properties {
  bAutomaticallySetUsageInEditor=false
  bUsedWithNiagaraMeshParticles=true
  blendMode=BLEND_Translucent
  shadingModel=MSM_Unlit
  usedShadingModels=Unlit
 }
 nodes [
  n1 = Panner(Coordinate=Add(A=TextureCoordinate() {vTiling=6}, B=NamedRerouteUsage())) {speedX=-0.3}
  n2 = Sine(None=ComponentMask(None=TextureCoordinate()) {g=true}) {period=2}
  n3 = LinearInterpolate(Alpha=LinearInterpolate(
    B=TextureSample(UVs=@n1) {samplerType=SAMPLERTYPE_Grayscale, texture=&LowResBlurredNoise}.R
    Alpha=Power(Base=@n2) {constExponent=3}
   )) {constA=0.4, constB=15}
  LinearInterpolate(Alpha=Saturate(None=@n3)) {constA=-1}
  n4 = StaticBool()
  n5 = ScalarParameter() {defaultValue=1, parameterName="Power (S)"}
  n6 = LinearInterpolate(A=LinearInterpolate(
    A=VectorParameter() {parameterName=4, defaultValue=<rgba 0.0625 0.0209 0.0071 1>}.RGB
    B=VectorParameter() {parameterName=3, defaultValue=<rgba 20 0.9412 0.9412 1>}.RGB
    Alpha=MaterialFunctionCall(Invert Fresnel=@n4, Power=@n5) {materialFunction=&Fresnel_Function}.Result
   ))
  n7 = TextureCoordinate() {vTiling=3}
  n8 = NamedRerouteUsage()
  n9 = TextureSample(UVs=Panner(Coordinate=Add(A=@n7, B=@n8)) {speedX=-0.5}) {samplerType=SAMPLERTYPE_Grayscale, texture=&LowResBlurredNoise}
  n10 = TextureCoordinate() {vTiling=2}
  n11 = NamedRerouteUsage()
  n12 = TextureSample(UVs=Panner(Coordinate=Add(A=@n10, B=@n11)) {speedX=0.5}) {samplerType=SAMPLERTYPE_Grayscale, texture=/BRCosmetics/Effects/Fort_Effects/Emotes/SwissKale/T_Blur_Noise}
  n13 = Saturate(None=LinearInterpolate(Alpha=LinearInterpolate(A=@n9.R, B=@n12.R) {constAlpha=0.3}) {constA=-0.3, constB=3})
  n14 = Multiply(A=Saturate(None=LinearInterpolate(Alpha=@n13) {constA=-1, constB=2}), B=Round(None=@n3))
  n15 = LinearInterpolate(
   A=VectorParameter() {parameterName=Color0, defaultValue=<rgba 0.15 0.0882 0 1>}.RGB
   B=VectorParameter() {parameterName=Color1, defaultValue=<rgba 2 0.2535 0.0742 1>}.RGB
   Alpha=Saturate(None=Add(B=@n14))
  )
  n16 = ParticleColor()
  MaterialFunctionCall(Emissive=Multiply(A=Multiply(A=@n6, B=@n15), B=@n16.RGB)) {materialFunction=&MF_FX_EmissiveAdjustments}
  n17 = TextureSample(UVs=Panner(Coordinate=TextureCoordinate() {vTiling=3}) {speedX=1.5}) {samplerType=SAMPLERTYPE_Grayscale, texture=&LowResBlurredNoise}
  n18 = MaterialFunctionCall(
   Transition Phase=DynamicParameter() {defaultValue=<rgba 1 1 0 1>, paramNames [Param1,Param2,Dissolve,Param4]}.Dissolve
   Height Texture=Saturate(None=LinearInterpolate(Alpha=@n17.R) {constA=0.3, constB=1.5})
  ) {materialFunction=&HeightLerp}
  Multiply(A=@n16.A, B=Saturate(None=Multiply(A=@n18.Results) {constB=3}))
  n19 = TextureSample(UVs=Panner(Coordinate=TextureCoordinate() {vTiling=3})) {samplerType=SAMPLERTYPE_Grayscale, texture=&LowResBlurredNoise}
  NamedRerouteDeclaration(None=Multiply(A=ComponentMask(None=@n19.RGB) {g=true, r=true}) {constB=0.3}) {name=Disturb, nodeColor=<rgba 1 0 0.3516 1>}
  Round(None=Multiply(A=@n13))
 ]
}
