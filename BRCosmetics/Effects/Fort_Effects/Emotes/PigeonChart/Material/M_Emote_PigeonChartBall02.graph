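# Translucent unlit material for Niagara mesh particles in the PigeonChart emote's ball effect, drawing animated, hard-edged noise shapes.
# Panning noise blends two colors, tinted by particle color through an emissive-adjustment function, with UVs offset by a noise-based Disturb reroute.
# Opacity combines two blended scrolling noises with a sine-shaped V mask, rounds them to a cutout, then applies particle alpha.
# A Dissolve dynamic parameter drives a HeightLerp erosion over fast-panning noise.
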Material /BRCosmetics/Effects/Fort_Effects/Emotes/PigeonChart/Material/M_Emote_PigeonChartBall02
graph {
 properties {
  bAutomaticallySetUsageInEditor=false
  bUsedWithNiagaraMeshParticles=true
  blendMode=BLEND_Translucent
  shadingModel=MSM_Unlit
  usedShadingModels=Unlit
 }
 nodes [
  n1 = VectorParameter() {parameterName=Color2, defaultValue=<rgba 0.4 0.4 0.4 1>}
  n2 = VectorParameter() {parameterName=Color1, defaultValue=<rgba 2 0 0 1>}
  n3 = TextureCoordinate() {vTiling=4}
  n4 = NamedRerouteUsage()
  n5 = TextureSample(UVs=Panner(Coordinate=Add(A=@n3, B=@n4)) {speedX=0.5}) {samplerType=SAMPLERTYPE_Grayscale, texture=/BRCosmetics/Effects/Fort_Effects/Emotes/SwissKale/T_Blur_Noise}
  n6 = ParticleColor()
  MaterialFunctionCall(Emissive=Multiply(A=LinearInterpolate(A=@n1.RGB, B=@n2.RGB, Alpha=@n5.R) {constA=0.4}, B=@n6.RGB)) {materialFunction=&MF_FX_EmissiveAdjustments}
  n7 = ScalarParameter() {defaultValue=-1, parameterName=EdgeMin, sortPriority=34}
  n8 = ScalarParameter() {defaultValue=2, parameterName=EdgeMax, sortPriority=35}
  n9 = Multiply(
   A=TextureCoordinate()
   B=ComponentMask(None=VectorParameter() {parameterName=NoiseTexA_UVScale, defaultValue=<rgba 1 4 0 0>}.RGB) {g=true, r=true}
  )
  n10 = NamedRerouteUsage()
  n11 = Constant2Vector() {r=0.5}
  n12 = ScalarParameter() {defaultValue=1, parameterName=Speed}
  n13 = TextureSampleParameter2D(UVs=Panner(Coordinate=Add(A=@n9, B=@n10), Speed=Multiply(A=@n11, B=@n12)) {speedX=0.5}) {parameterName=NoiseTexA, samplerType=SAMPLERTYPE_Grayscale, texture=&T_Blur_Noise}
  n14 = TextureCoordinate() {uTiling=2, vTiling=5}
  n15 = NamedRerouteUsage()
  n16 = Constant2Vector() {r=-0.5}
  n17 = ScalarParameter() {defaultValue=1, parameterName=Speed}
  n18 = TextureSampleParameter2D(UVs=Panner(Coordinate=Add(A=@n14, B=@n15), Speed=Multiply(A=@n16, B=@n17)) {speedX=-0.5}) {parameterName=NoiseTexB, texture=/BRCosmetics/Weapons/FORT_Melee/Pickaxe_CoatCheck/FX/T_Pickaxe_CoatCheck_Noise}
  n19 = ScalarParameter() {defaultValue=0.3, parameterName=TextureLerp}
  n20 = Saturate(None=LinearInterpolate(A=ScalarParameter() {defaultValue=-0.1, parameterName=DissolveMin}, Alpha=LinearInterpolate(A=@n13.R, B=@n18.R, Alpha=@n19) {constAlpha=0.3}) {constA=-0.1, constB=3})
  n21 = TextureCoordinate() {vTiling=6}
  n22 = NamedRerouteUsage()
  n23 = TextureSample(UVs=Panner(Coordinate=Add(A=@n21, B=@n22)) {speedX=-0.3}) {samplerType=SAMPLERTYPE_Grayscale, texture=&LowResBlurredNoise}
  n24 = TextureCoordinate()
  n25 = Power(Base=Sine(None=ComponentMask(None=@n24) {g=true}) {period=2}) {constExponent=3}
  n26 = Saturate(None=LinearInterpolate(Alpha=LinearInterpolate(B=@n23.R, Alpha=@n25)) {constA=0.4, constB=15})
  n27 = Multiply(
   A=Saturate(None=LinearInterpolate(A=@n7, B=@n8, Alpha=@n20) {constA=-1, constB=2})
   B=Saturate(None=LinearInterpolate(Alpha=@n26) {constA=-1})
  )
  n28 = LinearInterpolate(Alpha=Saturate(None=Add(B=@n27) {constB=0}))
  n29 = TextureSample(UVs=Panner(Coordinate=TextureCoordinate() {vTiling=3}) {speedX=1.5}) {samplerType=SAMPLERTYPE_Grayscale, texture=&LowResBlurredNoise}
  n30 = MaterialFunctionCall(
   Transition Phase=DynamicParameter() {defaultValue=<rgba 1 1 0 1>, paramNames [Param1,Param2,Dissolve,Param4]}.Dissolve
   Height Texture=Saturate(None=LinearInterpolate(Alpha=@n29.R) {constA=0.3, constB=1.5})
  ) {materialFunction=&HeightLerp}
  Multiply(A=Multiply(A=Round(None=@n28), B=@n6.A), B=Saturate(None=Multiply(A=@n30.Results) {constB=3}))
  n31 = TextureSample(UVs=Panner(Coordinate=TextureCoordinate() {vTiling=3})) {samplerType=SAMPLERTYPE_Grayscale, texture=&LowResBlurredNoise}
  NamedRerouteDeclaration(None=Multiply(A=ComponentMask(None=@n31.RGB) {g=true, r=true}) {constB=0.3}) {name=Disturb, nodeColor=<rgba 1 0 0.3516 1>}
  Round(None=@n5.R)
  TextureSample(UVs=Panner(Coordinate=TextureCoordinate() {uTiling=0.5, vTiling=3})) {samplerType=SAMPLERTYPE_Grayscale, texture=/BRCosmetics/Weapons/FORT_Melee/Pickaxe_Cube_Ninja_Male/FX/T_CellNoise}
 ]
}
