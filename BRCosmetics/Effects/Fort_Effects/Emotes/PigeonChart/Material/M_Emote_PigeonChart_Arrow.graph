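# Translucent, unlit, two-sided Niagara mesh-particle material for the arrow effect in the PigeonChart emote.
# Two panning noise layers blend three tint colors, multiplied by particle color for emissive.
# Opacity combines a noise-distorted scrolling shape mask, sine edge falloff with optional detail, view-angle fade and particle alpha.
# Dynamic parameters drive an Appear wipe along V and a vertex-color-weighted V offset, shared alongside UV disturbance through named reroutes.
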Material /BRCosmetics/Effects/Fort_Effects/Emotes/PigeonChart/Material/M_Emote_PigeonChart_Arrow
graph {
 properties {
  bAutomaticallySetUsageInEditor=false
  bUsedWithNiagaraMeshParticles=true
  blendMode=BLEND_Translucent
  shadingModel=MSM_Unlit
  twoSided=true
  usedShadingModels=Unlit
 }
 nodes [
  n1 = VectorParameter() {parameterName=TexB_UVScale, defaultValue=<rgba 1 0.7 0 0>}
  n2 = Add(A=Multiply(A=TextureCoordinate(), B=ComponentMask(None=@n1.RGB) {g=true, r=true}), B=NamedRerouteUsage())
  n3 = NamedRerouteUsage()
  n4 = VectorParameter() {group=Mask, parameterName=Mask_Speed, defaultValue=<rgba 0 1 0 0>}
  n5 = TextureSampleParameter2D(UVs=Panner(Coordinate=Add(A=@n2, B=@n3), Speed=ComponentMask(None=@n4.RGB) {g=true, r=true}) {speedY=0.5}) {parameterName=TexA, samplerType=SAMPLERTYPE_Grayscale, texture=&LowResBlurredNoise}
  n6 = LinearInterpolate(Alpha=Saturate(None=LinearInterpolate(Alpha=@n5.R))) {constA=-2, constB=3}
  n7 = LinearInterpolate(
   B=Multiply(A=Saturate(None=@n6), B=VectorParameter() {parameterName=Color1, defaultValue=<rgba 2 0.2883 0.1529 1>}.RGB)
   Alpha=ScalarParameter() {defaultValue=1, parameterName=AddTexA}
  )
  n8 = TextureCoordinate()
  n9 = ComponentMask(None=VectorParameter() {parameterName=TexB_UVScale, defaultValue=<rgba 1 0.7 0 0>}.RGB) {g=true, r=true}
  n10 = Add(A=Add(A=Multiply(A=@n8, B=@n9), B=NamedRerouteUsage()), B=NamedRerouteUsage())
  n11 = ComponentMask(None=VectorParameter() {group=Mask, parameterName=Mask_Speed, defaultValue=<rgba 0 1 0 0>}.RGB) {g=true, r=true}
  n12 = LinearInterpolate(
   A=ScalarParameter() {defaultValue=-0.3, parameterName=TexBMin, sortPriority=35}
   B=ScalarParameter() {defaultValue=2, parameterName=TexBMax, sortPriority=36}
   Alpha=TextureSampleParameter2D(UVs=Panner(Coordinate=@n10, Speed=@n11) {speedY=0.5}) {parameterName=TexB, samplerType=SAMPLERTYPE_Grayscale, texture=&LowResBlurredNoise}.R
  ) {constA=-0.3, constB=2}
  n13 = Saturate(None=LinearInterpolate(Alpha=Saturate(None=@n12)))
  n14 = LinearInterpolate(
   A=VectorParameter() {parameterName=Color3, defaultValue=<rgba 0.3 0.0467 0.0437 1>}.RGB
   B=VectorParameter() {parameterName=Color2, defaultValue=<rgba 0.7 0.1084 0.0316 1>}.RGB
   Alpha=@n13
  )
  n15 = ParticleColor()
  MaterialFunctionCall(Emissive=Multiply(A=Add(A=@n7, B=@n14), B=@n15.RGB)) {materialFunction=&MF_FX_EmissiveAdjustments}
  n16 = Constant() {r=1}
  n17 = Saturate(None=Multiply(A=MaterialFunctionCall("FadeSensitivity (optional)"=@n16) {materialFunction=&FadeBasedOnViewAngle}.Result))
  n18 = ScalarParameter() {parameterName=MaskMin, sortPriority=37}
  n19 = ScalarParameter() {defaultValue=1, parameterName=MaskMax, sortPriority=38}
  n20 = VectorParameter() {group=Mask, parameterName=Mask_Speed, defaultValue=<rgba 0 1 0 0>}
  n21 = Multiply(
   A=TextureCoordinate()
   B=ComponentMask(None=VectorParameter() {group=Mask, parameterName=Mask_UV_Scale, defaultValue=<rgba 1 1 1 1>}.RGB) {g=true, r=true}
  )
  n22 = NamedRerouteUsage()
  n23 = MaterialFunctionCall(Float2=Panner(Coordinate=Add(A=@n21, B=@n22), Speed=ComponentMask(None=@n20.RGB) {g=true, r=true})) {materialFunction=&BreakOutFloat2Components}
  n24 = MaterialFunctionCall(X=Multiply(A=@n23.R), Y=@n23.G) {materialFunction=&MakeFloat2}
  n25 = VectorParameter() {parameterName=Noise_ScaleUV, sortPriority=45, defaultValue=<rgba 0.2 0.4 0 0>}
  n26 = Add(A=Multiply(A=TextureCoordinate(), B=ComponentMask(None=@n25.RGB) {g=true, r=true}), B=NamedRerouteUsage())
  n27 = TextureSample(UVs=Panner(Coordinate=Multiply(A=@n26))) {samplerType=SAMPLERTYPE_LinearGrayscale, texture=/Game/Athena/Prototype/Terrain/T_CloudCoverage}
  n28 = TextureCoordinate()
  n29 = Multiply(
   A=Multiply(A=ComponentMask(None=@n27.RGB) {g=true, r=true}, B=ScalarParameter() {defaultValue=0.15, parameterName=MaskDisturb}) {constB=0.15}
   B=Power(Base=ComponentMask(None=@n28) {r=true}) {constExponent=0.8}
  )
  n30 = StaticComponentMaskParameter(None=TextureSampleParameter2D(UVs=Add(A=@n24.Result, B=@n29) {constB=0}) {group=Mask, parameterName=mask, texture=/BRCosmetics/Weapons/FORT_Melee/Pickaxe_Cube_Queen_Female/FX/T_CubeQueen_Spikes_Mask}.RGB) {defaultG=true, group=Mask, parameterName=Component_Mask}
  n31 = Add(A=Saturate(None=LinearInterpolate(A=@n18, B=@n19, Alpha=@n30))) {constB=0}
  n32 = TextureCoordinate()
  n33 = Power(Base=Sine(None=ComponentMask(None=@n32) {g=true}) {period=2}) {constExponent=1}
  n34 = TextureCoordinate()
  n35 = ComponentMask(None=VectorParameter() {group=Mask, parameterName=Mask_Speed, defaultValue=<rgba 0 1 0 0>}.RGB) {g=true, r=true}
  n36 = OneMinus(None=ComponentMask(None=TextureCoordinate()) {g=true})
  n37 = Power(
   Base=TextureSample(UVs=Panner(Coordinate=@n34, Speed=@n35) {speedY=0.5}) {samplerType=SAMPLERTYPE_LinearColor, texture=/Game/Effects/Fort_Effects/BluGlow/Violentium/Violentium_Mask}.R
   Exp=LinearInterpolate(Alpha=Power(Base=@n36) {constExponent=3}) {constB=5}
  )
  n38 = Multiply(A=@n33, B=Saturate(None=LinearInterpolate(Alpha=@n37) {constA=-1, constB=2}))
  n39 = ScalarParameter() {parameterName=UseEdgeDetailMask}
  n40 = Multiply(
   A=LinearInterpolate(B=@n13, Alpha=ScalarParameter() {parameterName=UseLessAlpha}) {constA=1}
   B=Multiply(A=Saturate(None=@n31), B=LinearInterpolate(A=@n33, B=@n38, Alpha=@n39))
  ) {constA=1}
  n41 = Add(
   A=ComponentMask(None=TextureCoordinate()) {g=true}
   B=LinearInterpolate(Alpha=DynamicParameter() {paramNames [V_Offset,Appear,Param3,Param4]}.Appear) {constA=-1}
  ) {constB=-1}
  Multiply(A=Multiply(A=@n15.A, B=Multiply(A=@n40, B=@n17)), B=Saturate(None=Multiply(A=@n41) {constB=3}))
  n42 = Add(A=TextureCoordinate() {vTiling=0.5}, B=NamedRerouteUsage())
  n43 = TextureSample(UVs=Panner(Coordinate=Multiply(A=@n42))) {samplerType=SAMPLERTYPE_Grayscale, texture=&LowResBlurredNoise}
  NamedRerouteDeclaration(None=Multiply(A=ComponentMask(None=@n43.RGB) {g=true, r=true}, B=ScalarParameter() {defaultValue=0.3, parameterName=DisturbIntensity}) {constB=0.3}) {name=Disturb, nodeColor=<rgba 0 0.8359 1 1>}
  n44 = Constant()
  n45 = DynamicParameter() {paramNames [V_Offset,Appear,Param3,Param4]}
  NamedRerouteDeclaration(None=LinearInterpolate(A=AppendVector(A=@n44, B=@n45.V_Offset), Alpha=VertexColor().R) {constB=0}) {name=VOffset, nodeColor=<rgba 0.2225 0 1 1>}
 ]
}
