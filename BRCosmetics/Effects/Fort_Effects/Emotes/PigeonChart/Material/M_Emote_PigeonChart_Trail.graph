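# Unlit, two-sided translucent trail material for the PigeonChart emote, built for Niagara ribbons and mesh particles.
# World position offset combines SplineThicken width tapering with a camera-transformed sine wave.
# Opacity multiplies a noise-distorted panning trail texture by end masks, an Appear wipe and HeightLerp dissolve, then depth-fades.
# Emissive blends two noise-modulated colors tinted by particle color; dynamic parameters drive width, reveal, dissolve and optional custom time, with optional hard edges.
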Material /BRCosmetics/Effects/Fort_Effects/Emotes/PigeonChart/Material/M_Emote_PigeonChart_Trail
graph {
 properties {
  bAutomaticallySetUsageInEditor=false
  bUsedWithNiagaraMeshParticles=true
  bUsedWithNiagaraRibbons=true
  blendMode=BLEND_Translucent
  shadingModel=MSM_Unlit
  twoSided=true
  usedShadingModels=Unlit
 }
 nodes [
  n1 = DynamicParameter() {defaultValue=<rgba 1 1 0 0>, paramNames [Width, Appear, Dissolve, SmallTrailOffset]}
  n2 = Multiply(A=ScalarParameter() {defaultValue=20, parameterName=WidthBase}, B=@n1.Width)
  n3 = Power(Base=ComponentMask(None=TextureCoordinate()) {r=true}, Exp=ScalarParameter() {defaultValue=1, parameterName=Top WPO Fade})
  n4 = TextureCoordinate()
  n5 = Power(Base=OneMinus(None=ComponentMask(None=@n4) {r=true})) {constExponent=3}
  n6 = LinearInterpolate(
   A=Saturate(None=Multiply(A=@n3) {constB=3})
   B=LinearInterpolate(Alpha=Multiply(A=@n5) {constB=3}) {constA=1, constB=1.5}
   Alpha=ScalarParameter() {parameterName=UseWidthMaskInvert}
  )
  n7 = Multiply(A=ScalarParameter() {defaultValue=20, parameterName=WidthTip}, B=@n1.Width)
  n8 = ComponentMask(None=TextureCoordinate()) {r=true}
  n9 = Multiply(A=Time()) {constB=-0.5}
  n10 = Multiply(
   A=Sine(None=Add(A=@n8, B=@n9)) {period=0.8}
   B=Transform(None=VectorParameter() {parameterName=WPO Direction, defaultValue=<rgba 0 0 17 0>}.RGB) {transformType=TRANSFORM_Camera}
  )
  n11 = DynamicParameter() {parameterIndex=2, defaultValue=<rgba 1 1 0 1>, paramNames [WPOIntensity,Time,None,None]}
  n12 = TextureCoordinate()
  n13 = Multiply(A=Power(Base=ComponentMask(None=@n12) {r=true}))
  Add(
   A=MaterialFunctionCall(WidthBase=Multiply(A=@n2, B=@n6), WidthTip=Multiply(A=@n7, B=@n6), Expand U or V UV Channel=StaticBool()) {materialFunction=&SplineThicken}.WorldPositionOffset
   B=Multiply(A=Multiply(A=@n10, B=@n11.WPOIntensity), B=Saturate(None=@n13))
  )
  n14 = ParticleColor()
  n15 = TextureCoordinate()
  n16 = Power(Base=ComponentMask(None=@n15) {r=true}) {constExponent=1.2}
  n17 = ComponentMask(None=@n15) {g=true}
  n18 = VectorParameter() {parameterName=AlphaUVScale, defaultValue=<rgba 3 1 0 0>}
  n19 = Add(
   A=Multiply(A=AppendVector(A=@n16, B=@n17), B=ComponentMask(None=@n18.RGB) {g=true, r=true})
   B=ComponentMask(None=VectorParameter() {parameterName=AlphaUVOffset, sortPriority=40}.RGB) {g=true, r=true}
  )
  n20 = NamedRerouteUsage()
  n21 = Time()
  n22 = DynamicParameter() {parameterIndex=2, defaultValue=<rgba 1 1 0 1>, paramNames [WPOIntensity,Time,None,None]}
  n23 = ScalarParameter() {parameterName=UseCustomTime}
  n24 = Constant2Vector() {r=1}
  n25 = ScalarParameter() {defaultValue=1, parameterName=PannerSpeed}
  n26 = TextureSampleParameter2D(UVs=Panner(Coordinate=Add(A=@n19, B=@n20), Time=LinearInterpolate(A=@n21, B=@n22.Time, Alpha=@n23) {constB=0}, Speed=Multiply(A=@n24, B=@n25)) {speedX=1}) {
   parameterName=Alpha
   texture=/BRCosmetics/Weapons/FORT_Melee/Pickaxe_Future_Samurai_Summer_Male/FX/T_Pickaxe_Future_Samurai_Summer_Male_Trail_01
  }
  n27 = SmoothStep(
   Min=ScalarParameter() {defaultValue=0.3, parameterName=AlphaLerpMin, sortPriority=33}
   Max=ScalarParameter() {defaultValue=0.8, parameterName=AlphaLerpMax, sortPriority=34}
   Value=@n26.R
  ) {constMax=0.8, constMin=0.3}
  n28 = Round(None=Multiply(A=@n27) {constB=2})
  n29 = Multiply(A=@n14.A, B=LinearInterpolate(A=@n27, B=Saturate(None=@n28), Alpha=ScalarParameter() {parameterName=Use HardEdge}))
  n30 = OneMinus(None=ComponentMask(None=TextureCoordinate()) {r=true})
  n31 = ScalarParameter() {defaultValue=2, parameterName=AlphaPower}
  n32 = Saturate(None=Multiply(A=Power(Base=@n30, Exp=@n31)) {constB=6})
  n33 = ComponentMask(None=TextureCoordinate()) {r=true}
  n34 = OneMinus(None=@n33)
  n35 = ScalarParameter() {parameterName=UseInvertLeadMask}
  n36 = DynamicParameter() {defaultValue=<rgba 1 1 0 0>, paramNames [Width, Appear, Dissolve, SmallTrailOffset]}
  n37 = Multiply(A=Add(A=LinearInterpolate(A=@n33, B=@n34, Alpha=@n35), B=LinearInterpolate(Alpha=@n36.Appear) {constA=-1}) {constB=-1}) {constB=5}
  n38 = Add(A=TextureCoordinate() {uTiling=2}, B=NamedRerouteUsage())
  n39 = MaterialFunctionCall(
   Transition Phase=LinearInterpolate(Alpha=DynamicParameter() {defaultValue=<rgba 1 1 0 0>, paramNames [Width, Appear, Dissolve, SmallTrailOffset]}.Dissolve) {constA=0.5}
   Height Texture=TextureSample(UVs=Panner(Coordinate=@n38)) {samplerType=SAMPLERTYPE_Grayscale, texture=&LowResBlurredNoise}.R
  ) {materialFunction=&HeightLerp}
  n40 = Multiply(A=Multiply(A=Multiply(A=@n29, B=@n32), B=Saturate(None=@n37)), B=Saturate(None=Multiply(A=@n39.Results) {constB=3}))
  n41 = TextureCoordinate()
  n42 = Power(Base=Add(A=ComponentMask(None=@n41) {r=true}, B=ScalarParameter() {parameterName=LeadMaskOffset})) {constExponent=3}
  n43 = ScalarParameter() {defaultValue=200, parameterName=LeadMaskIntensity}
  n44 = LinearInterpolate(B=Saturate(None=Multiply(A=@n42, B=@n43) {constB=200}), Alpha=ScalarParameter() {parameterName=UseleadMask}) {constA=1}
  n45 = DynamicParameter() {defaultValue=<rgba 1 1 0 0>, paramNames [Width, Appear, Dissolve, SmallTrailOffset]}
  n46 = TextureSample(UVs=Panner(Coordinate=TextureCoordinate() {uTiling=1.5}, Time=LinearInterpolate(Alpha=@n45.SmallTrailOffset) {constA=-2}) {speedX=1}) {
   samplerSource=SSM_Clamp_WorldGroupSettings
   samplerType=SAMPLERTYPE_LinearGrayscale
   texture=/Game/Gadgets/Assets/VinderTech_GliderChute/Glider_Ensemble_Maroon_Male/FX/Textures/T_RibbonMask_B
  }
  n47 = ScalarParameter() {parameterName=UseSmallTrailTexForAlpha}
  DepthFade(
   Opacity=Multiply(A=Multiply(A=@n40, B=@n44), B=LinearInterpolate(B=@n46.R, Alpha=@n47) {constA=1})
   FadeDistance=ScalarParameter() {defaultValue=10, parameterName=FadeDistance}
  )
  DynamicParameter() {defaultValue=<rgba 1 1 0 0>, paramNames [Width, Appear, Dissolve, SmallTrailOffset]}
  LinearInterpolate(Alpha=@n39.Results) {constA=-1, constB=3}
  n48 = ScalarParameter() {defaultValue=-5, parameterName=LerpAMin, sortPriority=1}
  n49 = ScalarParameter() {defaultValue=20, parameterName=LerpAMax, primitiveDataIndex=1, sortPriority=2}
  n50 = TextureCoordinate() {uTiling=2}
  n51 = NamedRerouteUsage()
  n52 = Time()
  n53 = DynamicParameter() {parameterIndex=2, defaultValue=<rgba 1 1 0 1>, paramNames [WPOIntensity,Time,None,None]}
  n54 = ScalarParameter() {parameterName=UseCustomTime}
  n55 = Constant2Vector() {r=0.3}
  n56 = ScalarParameter() {defaultValue=1, parameterName=PannerSpeed}
  n57 = TextureSampleParameter2D(UVs=Panner(
    Coordinate=Add(A=@n50, B=@n51) {constB=0}
    Time=LinearInterpolate(A=@n52, B=@n53.Time, Alpha=@n54) {constB=0}
    Speed=Multiply(A=@n55, B=@n56)
   ) {speedX=0.3}) {parameterName=LerpMinTex, samplerType=SAMPLERTYPE_Grayscale, sortPriority=7, texture=&LowResBlurredNoise}
  n58 = ScalarParameter() {defaultValue=1, parameterName=LerpBMin, primitiveDataIndex=2, sortPriority=3}
  n59 = ScalarParameter() {defaultValue=4, parameterName=LerpBMax, primitiveDataIndex=3, sortPriority=4}
  n60 = TextureCoordinate() {uTiling=2}
  n61 = NamedRerouteUsage()
  n62 = Time()
  n63 = DynamicParameter() {parameterIndex=2, defaultValue=<rgba 1 1 0 1>, paramNames [WPOIntensity,Time,None,None]}
  n64 = ScalarParameter() {parameterName=UseCustomTime}
  n65 = Constant2Vector() {r=0.6}
  n66 = ScalarParameter() {defaultValue=1, parameterName=PannerSpeed}
  n67 = TextureSampleParameter2D(UVs=Panner(
    Coordinate=Add(A=@n60, B=@n61) {constB=0}
    Time=LinearInterpolate(A=@n62, B=@n63.Time, Alpha=@n64) {constB=0}
    Speed=Multiply(A=@n65, B=@n66)
   ) {speedX=0.6}) {
   parameterName=LerpMaxTex
   samplerType=SAMPLERTYPE_Grayscale
   sortPriority=8
   texture=/BRCosmetics/Effects/Fort_Effects/Emotes/SwissKale/T_Blur_Noise
  }
  n68 = Multiply(
   A=TextureCoordinate() {uTiling=2.5, vTiling=1.2}
   B=ComponentMask(None=VectorParameter() {parameterName=ColorRangeTexUVScale, sortPriority=42, defaultValue=<rgba 1 1 0 0>}.RGB) {g=true, r=true}
  )
  n69 = Time()
  n70 = DynamicParameter() {parameterIndex=2, defaultValue=<rgba 1 1 0 1>, paramNames [WPOIntensity,Time,None,None]}
  n71 = ScalarParameter() {parameterName=UseCustomTime}
  n72 = Constant2Vector() {r=1}
  n73 = ScalarParameter() {defaultValue=1, parameterName=PannerSpeed}
  n74 = TextureSampleParameter2D(UVs=Panner(Coordinate=Add(A=@n68) {constB=0}, Time=LinearInterpolate(A=@n69, B=@n70.Time, Alpha=@n71) {constB=0}, Speed=Multiply(A=@n72, B=@n73)) {speedX=1}) {
   parameterName=ColorRangeTex
   samplerType=SAMPLERTYPE_Grayscale
   sortPriority=9
   texture=/Game/Effects/Fort_Effects/Textures/Smoke/T_Smoke_Trail_Soft_NoClamp
  }
  n75 = LinearInterpolate(
   A=ScalarParameter() {defaultValue=-2, parameterName=LerpCMin, primitiveDataIndex=4, sortPriority=5}
   B=ScalarParameter() {defaultValue=3, parameterName=LerpCMax, primitiveDataIndex=5, sortPriority=6}
   Alpha=Saturate(None=LinearInterpolate(Alpha=@n74.R) {constA=-0.5})
  ) {constA=-2, constB=4}
  n76 = LinearInterpolate(
   A=Multiply(
    A=VectorParameter() {parameterName=Color0, defaultValue=<rgba 0.3 0.025 0 1>}.RGB
    B=LinearInterpolate(A=@n48, B=@n49, Alpha=@n57.R) {constA=0.3, constB=3}
   )
   B=Multiply(
    A=LinearInterpolate(A=@n58, B=@n59, Alpha=@n67.R) {constB=4}
    B=VectorParameter() {parameterName=Color1, defaultValue=<rgba 0.8 0.096 0.003 1>}.RGB
   )
   Alpha=LinearInterpolate(A=Saturate(None=@n75), B=Round(None=@n75), Alpha=ScalarParameter() {parameterName=Use HardEdge})
  )
  MaterialFunctionCall(Emissive=LinearInterpolate(A=Multiply(A=@n76, B=@n14.RGB), B=@n14.RGB, Alpha=ScalarParameter() {parameterName=UseSimpleParticleColor})) {materialFunction=&MF_FX_EmissiveAdjustments}
  MaterialFunctionCall() {materialFunction=&TimeWithSpeedVariable}
  MaterialFunctionCall() {materialFunction=&DebugTimeSine}
  MaterialFunctionCall() {materialFunction=&DebugTimeSine}
  n77 = TextureSample(UVs=Panner(Coordinate=TextureCoordinate() {vTiling=0.5}) {speedX=0.1, speedY=0.1}) {samplerType=SAMPLERTYPE_LinearGrayscale, texture=/Game/Effects/Fort_Effects/Emotes/UltraLight/Texture/T_Cloud_Greyscale}
  NamedRerouteDeclaration(None=Multiply(A=ComponentMask(None=@n77.RGB) {g=true, r=true}, B=ScalarParameter() {defaultValue=0.21, parameterName=Distorb, sortPriority=39}) {constB=0.21}) {name=Noise, nodeColor=<rgba 1 0.0469 0 1>}
  ScalarParameter() {defaultValue=1, parameterName=Top WPO Fade}
 ]
}
